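# M_Ruckus_Decal is a masked, two-sided Default Lit material, named as the decal material for the Ruckus vehicle cosmetic body.
# It holds two texture samples, a base color and a normal map, both borrowed from the Alokin body's chassis decal textures.
# The graph shows no links from these samples to material outputs.
# Usage is enabled for skeletal meshes and mesh deformers, with automatic usage setting in the editor turned off.
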Material /VehicleCosmetics/Bodies/Ruckus/Materials/M_Ruckus_Decal
graph {
 properties {
  bAutomaticallySetUsageInEditor=false
  bUsedWithMeshDeformer=true
  bUsedWithSkeletalMesh=true
  blendMode=BLEND_Masked
  twoSided=true
  usedShadingModels=Default Lit
 }
 nodes [
  TextureSample() {texture=/VehicleCosmetics/Bodies/Alokin/Textures/Chassis/T_Alokin_Decal_BC}
  TextureSample() {samplerType=SAMPLERTYPE_Normal, texture=/VehicleCosmetics/Bodies/Alokin/Textures/Chassis/T_Alokin_Decal_N}
 ]
}
